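# Sound cue for the present-perk gadget's drop landing, layering a box impact with a jingle.
# A mixer combines two modulated branches, each randomly picking one of four equally weighted variations: plain drop waves and jingle drop waves.
# The jingle branch passes through a tiny random delay so it slightly trails or coincides with the impact.
# It plays in the impacts sound class with weapon-drop attenuation and reduced volume.
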SoundCue /Game/Sounds/Fort_Gadgets/Present_Perk/Presnt_Perk_Drop_Land_Cue
root=Mixer_0
nodes [
 {id=Delay_0, children [Modulator_1], properties {delayMax=0.02}}
 {id=Mixer_0, children [Modulator_0,Delay_0], properties {inputVolume [1,1]}}
 {id=Modulator_0, children [Random_0]}
 {id=Modulator_1, children [Random_3]}
 {id=Random_0, children [WavePlayer_1,WavePlayer_2,WavePlayer_3,WavePlayer_4], properties {weights [1,1,1,1]}}
 {id=Random_3, children [WavePlayer_6, WavePlayer_12, WavePlayer_13, WavePlayer_14], properties {weights [1,1,1,1]}}
 {id=WavePlayer_1, properties {soundWaveAssetPtr=&/Box_Presents_Drop_04}}
 {id=WavePlayer_12, properties {soundWaveAssetPtr=&/Box_Presents_Jingle_Drop_06}}
 {id=WavePlayer_13, properties {soundWaveAssetPtr=&/Box_Presents_Jingle_Drop_07}}
 {id=WavePlayer_14, properties {soundWaveAssetPtr=&/Box_Presents_Jingle_Drop_08}}
 {id=WavePlayer_2, properties {soundWaveAssetPtr=&/Box_Presents_Drop_01}}
 {id=WavePlayer_3, properties {soundWaveAssetPtr=&/Box_Presents_Drop_02}}
 {id=WavePlayer_4, properties {soundWaveAssetPtr=&/Box_Presents_Drop_03}}
 {id=WavePlayer_6, properties {soundWaveAssetPtr=&/Box_Presents_Jingle_Drop_05}}
]
properties {
 attenuationSettings=&Fort_Weapon_Drop_Att
 maxDistance=3000
 soundClassObject=&Fort_Impacts
 volumeMultiplier=0.6
}
